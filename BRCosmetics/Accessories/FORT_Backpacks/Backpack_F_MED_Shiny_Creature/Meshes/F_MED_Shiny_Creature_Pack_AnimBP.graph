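# Animation Blueprint for the Shiny Creature backpack cosmetic, built to offset the pack's root bone as the wearer's spine bends.
# Outside the front end, each update maps the negative yaw delta of spine_05 against spine_01 into clamped roll, pitch and Y translation.
# The AnimGraph chains three additive root-bone modifiers on the reference pose: two fixed Y nudges around the spine-driven one, whose alpha is zero, leaving it inactive.
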AnimBlueprint /BRCosmetics/Accessories/FORT_Backpacks/Backpack_F_MED_Shiny_Creature/Meshes/F_MED_Shiny_Creature_Pack_AnimBP
target_skeleton=&/F_MED_Shiny_Creature_Pack_Skeleton
graphs [
 AnimGraph {nodes [
   n1 = "Transform (Modify) Bone - Bone: root"(ComponentPose=Local To Component(LocalPose=Local Space Ref Pose()), Translation="0,1,0", Rotation="0,0,0") {translationMode=BMM_Additive, alphaBoolBlend {blendOption=Linear}, boneToModify {boneName=root}, translation {y=1}}
   n2 = "Transform (Modify) Bone - Bone: root"(ComponentPose=@n1, Translation=Get backpack_translation(), Rotation=Get backpack_rotate()) {alpha=0, rotationMode=BMM_Additive, translationMode=BMM_Additive, alphaBoolBlend {blendOption=Linear}, boneToModify {boneName=root}}
   n3 = "Transform (Modify) Bone - Bone: root"(ComponentPose=@n2, Translation="0,1,0", Rotation="0,0,0") {translationMode=BMM_Additive, alphaBoolBlend {blendOption=Linear}, boneToModify {boneName=root}, translation {y=1}}
   Output Pose(Result=Component To Local(ComponentPose=@n3)) {layerGroup=None}
  ]}
 EventGraph {nodes [
   Event BlueprintUpdateAnimation()
   > n1 = Branch(Condition=Get bIsInFrontEnd())
   n2 = GetOwnerSkeletalMeshComponent()
   n3 = BreakRotator(InRot=BreakTransform(InTransform=GetDeltaTransformFromRefPose(self=@n2, BoneName=spine_05, BaseName=spine_01)).Rotation)
   Set backpack_rotate(
    execute=@n1.else
    backpack_rotate=MakeRotator(
     Roll=MapRangeClamped(Value=@n3.Yaw, InRangeA=0, InRangeB=-35, OutRangeB=40)
     Pitch=MapRangeClamped(Value=@n3.Yaw, InRangeA=0, InRangeB=-25, OutRangeB=10)
    )
   )
   > Set backpack_translation(backpack_translation=MakeVector(Y=MapRangeClamped(Value=@n3.Yaw, InRangeA=0, InRangeB=-16, OutRangeB=-4)))
   TryGetPawnOwner()
  ]}
]
variables [
 backpack_rotate {type=Rotator}
 backpack_translation {type=Vector}
]
